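# Idle Niagara effect for the FoundKelp Bolt Castfall character cosmetic, combining floating shards, reactive heat glow and ice bursts on sockets.
# Persistent shards bob in eight socket partitions, pushed outward by User.ReactiveCurve.
# That curve also drives glows, smoke, heat distortion and lights, with cyan-to-red-and-back tinting on glows and smoke.
# UpdateValueStateChange gates falling ice shard and smoke bursts, while hand/foot-hidden flags zero selected socket elements.
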NiagaraSystem /BRCosmetics/Characters/Player/Male/Medium/Bodies/M_MED_FoundKelp_Cast/FX/NS_Character_FoundKelp_Bolt_Castfall_Idle
dependencies {
 usedDataInterfaces [
  &NiagaraDataInterfaceCurve
  &NiagaraDataInterfaceArrayFloat
  &NiagaraDataInterfaceArrayFloat3
  &NiagaraDataInterfaceArrayInt32
  &NiagaraDataInterfaceSpriteRendererInfo
  &NiagaraDataInterfaceColorCurve
  &NiagaraDataInterfaceSocketReader
 ]
 usedDynamicInputs []
 usedModules [
  &self:SystemSpawnScript.SystemScriptSource.SystemScriptGraph.NiagaraNodeAssignment_3.SetVariables_#1
  &self:SystemSpawnScript.SystemScriptSource.SystemScriptGraph.NiagaraNodeAssignment_0.SetVariables_#2
  &SystemState
  /Game/Effects/Niagara/Modules/Logic/UpdateValueStateChange
 ]
}
emitters {
 FlyingShards {
  emitterUpdateScript [
   &EmitterState(Life Cycle Mode=::NewEnumerator1, Loop Behavior=::NewEnumerator1, Loop Duration=1, UseLoopDelay=false)
   &SpawnPerFrame(
    Spawn:NiagaraBool=&SetBoolByFloatComparison(A=0, B=1, Comparison Type=::NewEnumerator4)
    Spawn Count=20
    Use Spawn Probability=false
   ) {enabled=false}
   &SpawnPerFrame(
    Spawn:NiagaraBool=&Boolean_AndOperation(
     Bool1=&Boolean_AndOperation(
      Bool1=&Boolean_NotOperation(Bool=$System.UpdateValueStateChange.Changed)
      Bool2=&Boolean_NotOperation(Bool=$Transient.FirstFrame)
     )
     Bool2=true
    )
    Spawn Count=20
    Use Spawn Probability=false
   ) {enabled=false, module=SpawnPerFrame001}
   &SpawnBurst_Instantaneous(Spawn Count=40, Use Spawn Probability=false, UseLoopCountLimit=false)
  ]
  particleSpawnScript [
   &PartitionParticles(Number Of Partitions=8, Particles Per Partition=5, Partition Mode:ENiagara_PartitionMode_v2=::NewEnumerator1)
   /SetVariables(Particles.MeshIndex=&RandomRangeInt(Minimum=0, Maximum=2)) {isSetParameters=true, module=SetVariables_#3}
   /SetVariables(Particles.SocketIndex:NiagaraInt32=&SelectIntFromArray(
     Direct Array Index=$Particles.PartitionParticles.Partition
     Int Selection Array=DataInterface() {propertyValues {IntData [1,2,3,4,5,6,16,17]}}
    )) {enabled=false, isSetParameters=true, module=SetVariables_#4}
   &InitializeParticle(
    Lifetime=&RandomRangeFloat(Minimum=0.8, Maximum=1.2)
    Mesh Scale=&Multiply_VectorByFloat(Vector=<1 1 1>, Float=&RandomRangeFloat(Minimum=0.9, Maximum=1.1))
    Mesh Scale Mode=::NewEnumerator4
    Sprite Size Mode=::NewEnumerator3
    Uniform Sprite Size=1
    UsePositionOffset=false
    Write Material Random=true
   )
   &InitialMeshOrientation(
    Mesh Orientation Mode=::NewEnumerator2
    Orientation Coordinate Space=::Local
    Orientation Vector=&MakeVector(
     X=&SelectFloatFromArray(
      Array Sampling Mode=::NewEnumerator0
      Direct Array Index=?
      Float Selection Array=DataInterface() {propertyValues {FloatData [-1,1]}}
      Random Seed=0
     )
     Y=&RandomRangeFloat(Minimum=-0.3, Maximum=0.3)
     Z=&RandomRangeFloat(Minimum=-0.3, Maximum=0.3)
    )
    Use Rotation Vector=false
   )
   &AddVelocity(Velocity=&Multiply_VectorByFloat(Vector=$Particles.ShapeLocation.ShapeNormal, Float=&RandomRangeFloat(Minimum=150, Maximum=250))) {enabled=false}
  ]
  particleUpdateScript [
   &ParticleState(Kill Particles When Lifetime Has Elapsed=false)
   &SocketLocation(
    SamplePosOffset=true
    Sampled Orientation Coordinate Space:ENiagaraCoordinateSpace=::Local
    Sampled Position Offset=&Add_Vector(
     A=&Lerp_Vector(
      A=<0 0 0>
      B=&MergeHelper_VectorSwitch(
       0=&RandomRangeVector(Minimum=<0 0 0>, Maximum=<0 0 -10>)
       1=&RandomRangeVector(Minimum=<0 0 0>, Maximum=<0 0 -10>)
       2=&RandomRangeVector(Minimum=<0 0 0>, Maximum=<0 0 10>)
       3=&RandomRangeVector(Minimum=<0 0 0>, Maximum=<0 0 10>)
       4=&RandomRangeVector(Minimum=<0 0 0>, Maximum=<0 0 -10>)
       5=&RandomRangeVector(Minimum=<0 0 0>, Maximum=<0 0 10>)
       6=<0 0 0>
       7=<0 0 0>
       Index=$Particles.PartitionParticles.Partition
       Num=::NewEnumerator6
      )
      Alpha=&FloatFromCurve(
       FloatCurve=DataInterface() {propertyValues {Curve {
          defaultValue=0
          keys [
           {arriveTangent=29.3281, interpMode=RCIM_Cubic, leaveTangent=29.3281, tangentMode=RCTM_User, time=0, value=0}
           {interpMode=RCIM_Cubic, tangentMode=RCTM_User, time=0.056, value=1}
           {arriveTangent=-2.9138, tangentMode=RCTM_User, time=0.829, value=0}
          ]
         }}}
       CurveIndex=$User.ReactiveCurve
      )
     )
     B=&Add_Vector(
      A=&MakeVector(
       X=&Sine(Normalized Angle=$Emitter.Age, Period=&RandomRangeFloat(Minimum=2, Maximum=4), Scale=1, Bias=0)
       Y=&Sine(Normalized Angle=$Emitter.Age, Period=&RandomRangeFloat(Minimum=2, Maximum=4), Scale=1, Bias=0)
       Z=&Sine(Normalized Angle=$Emitter.Age, Period=&RandomRangeFloat(Minimum=2, Maximum=4), Scale=1, Bias=0)
      )
      B=&SelectVectorFromArray(
       Direct Array Index=&ReturnExecIndex()
       Vector Selection Array=DataInterface() {propertyValues {FloatData [
          <14.0612 0 -8.6365>
          <4.2451 -4.2694 -5.5033>
          <8.79 3.7837 -7.823>
          <0.5327 6.9889 -3.6792>
          <0.4921 3.9762 -4.8126>
          <17.049 -1.4795 -6.855>
          <14.5628 3.4017 -6.0905>
          <14.9448 -4.5796 -9.261>
          <9.3694 -6.7933 -4.1539>
          <18.8018 4.8688 -5.5218>
          <-12.9776 -6.7933 1.1239>
          <-14.379 -0.0289 4.7485>
          <-9.2801 3.921 4.3792>
          <-6.7989 -2.8022 5.2084>
          <-8.1465 -5.9858 4.0716>
          <-21.6439 4.1517 5.7903>
          <-17.4873 7.4647 9.0027>
          <-14.8111 13.8679 5.5886>
          <-10.6383 0.2957 5.8325>
          <-10.9148 8.3893 6.889>
          <9.5996 -2.0269 -1.7355>
          <3.1982 2.0192 -4.4183>
          <7.6657 5.2054 0.965>
          <3.9367 -5.899 -5.399>
          <0.6899 8.8617 -0.2044>
          <-12.1124 6.9133 -1.3382>
          <-3.4302 8.6721 -1.8421>
          <-8.0752 3.3591 2.939>
          <-3.5521 -1.0664 4.2746>
          <0.345 -7.3462 4.2482>
          <11.2793 -10.0408 3.492>
          <6.2358 -9.9663 7.6558>
          <5.3039 -7.947 -2.8>
          <10.3504 -3.1298 -5.6319>
          <1.3942 -12.4946 5.9642>
          <-11.3374 6.1422 2.3537>
          <-4.9834 9.1839 -10.8814>
          <-3.6556 9.7274 -0.9205>
          <-8.663 9.1458 -7.7825>
          <-1.0854 4.7755 3.1795>
         ]}}
      )
     )
    )
    Socket Index=$Particles.PartitionParticles.Partition
    Socket Reader=$System.Shards_Sockets
    Socket Sampling Mode=::NewEnumerator3
    VelocitySpeedLimit=false
   )
   &ScaleMeshSize(Scale Factor=&Multiply_VectorByFloat(
     Vector=<1 1 1>
     Float=&Lerp_Float(
      A=1
      B=&MergeHelper_FloatSwitch(0=1, 1=0, 2=1, 3=0, 4=1, 5=1, 6=1, 7=1, Index=$Particles.PartitionParticles.Partition, Num=::NewEnumerator6)
      Alpha=&MakeCustomFloatFromBool(Bool=$User.bisHandHidden, True Float=1, False Float=0)
     )
    )) {module=ScaleMeshSize001}
   &ScaleMeshSize(Scale Factor=&Multiply_VectorByFloat(
     Vector=<1 1 1>
     Float=&Lerp_Float(
      A=1
      B=&MergeHelper_FloatSwitch(0=1, 1=1, 2=1, 3=1, 4=1, 5=1, 6=0, 7=0, Index=$Particles.PartitionParticles.Partition, Num=::NewEnumerator6)
      Alpha=&MakeCustomFloatFromBool(Bool=$User.bisFootHidden, True Float=1, False Float=0)
     )
    ))
   /SetVariables(Particles.MeshOrientation=$Output.SocketLocation.SampledRotation) {enabled=false, isSetParameters=true}
   &UpdateMeshOrientation(
    Facing Direction=&RotateVectorByQuaternion(
     Vector To Rotate=&MakeVector(
      X=&MergeHelper_FloatSwitch(
       0=&RandomRangeFloat(Minimum=0, Maximum=0.5)
       1=&RandomRangeFloat(Minimum=-0.5, Maximum=0.5)
       2=&RandomRangeFloat(Minimum=0, Maximum=0.75)
       3=&RandomRangeFloat(Minimum=0, Maximum=0.75)
       4=&RandomRangeFloat(Minimum=0.25, Maximum=0.75)
       5=&RandomRangeFloat(Minimum=0.25, Maximum=0.75)
       6=&RandomRangeFloat(Minimum=0.25, Maximum=0.75)
       7=&RandomRangeFloat(Minimum=0.25, Maximum=0.75)
       Index=$Particles.PartitionParticles.Partition
       Num=::NewEnumerator6
      )
      Y=0
      Z=&SelectFloatFromArray(
       Array Sampling Mode=::NewEnumerator0
       Direct Array Index=?
       Float Selection Array=DataInterface() {propertyValues {FloatData [1,1]}}
       Random Seed=0
      )
     )
     Quaternion=$StackContext.MeshOrientation
    )
    Facing Interpolation=false
    Orientation Method=::NewEnumerator1
    Override Quat=false
    TransformFacing=false
    TransformUp=false
   ) {enabled=false}
   &UpdateMeshOrientation(
    Rotation Coordinate Space=::NewEnumerator2
    Rotation Rate=&RandomRangeFloat(Minimum=0, Maximum=10)
    Rotation Vector=&RotateVectorByQuaternion(Vector To Rotate=<0 0 1>, Quaternion=$Output.UpdateMeshOrientation.MeshOrientation)
   ) {enabled=false, module=UpdateMeshOrientation001}
   &OffsetPosition(
    Jitter Amount:NiagaraFloat=&FloatFromCurve(
     FloatCurve=DataInterface() {propertyValues {Curve {
        defaultValue=0
        keys [
         {interpMode=RCIM_Cubic, tangentMode=RCTM_User, time=0, value=0}
         {interpMode=RCIM_Cubic, tangentMode=RCTM_User, time=0.0789, value=1}
         {arriveTangent=-3.422, tangentMode=RCTM_User, time=0.8387, value=0}
        ]
       }}}
     CurveIndex=$User.ReactiveCurve
     Scale Curve=0.25
    )
    Jitter Mode=1
    Jitter Offset Coordinate Space:ENiagaraCoordinateSpace=::Local
    Offset Mode=1
   ) {enabled=false}
  ]
  rendererClasses [&NiagaraMeshRendererProperties]
  renderers [
   {rendererClass=&NiagaraMeshRendererProperties, rendererIndex=0}
   {rendererClass=&NiagaraMeshRendererProperties, rendererIndex=1}
  ]
 }
 Glow_FrontView {
  emitterUpdateScript [
   &EmitterState(Life Cycle Mode=::NewEnumerator1, Loop Behavior=::NewEnumerator1, Loop Duration=1, UseLoopDelay=false)
   &SpawnBurst_Instantaneous(Spawn Group=2, Use Spawn Probability=false, UseLoopCountLimit=false) {module=SpawnBurst_Instantaneous002}
   &SpawnBurst_Instantaneous(Use Spawn Probability=false, UseLoopCountLimit=false)
   &SpawnBurst_Instantaneous(Spawn Group=1, Use Spawn Probability=false, UseLoopCountLimit=false) {module=SpawnBurst_Instantaneous001}
  ]
  particleSpawnScript [&InitializeParticle(
    Color=&MakeCustomLinearColorFromBool(Bool=$User.bIsFrontEndPreview, True Color=<rgba 1 1 1 1>, False Color=<rgba 1 1 1 2.5>)
    Lifetime=5
    Sprite Size=&MaskVector2dBySpawnGroup(
     Masked Vector2D=<46.1749 26.2674>
     Passthrough Vector2D=&MaskVector2dBySpawnGroup(Masked Vector2D=<35.6836 20.2351>, Passthrough Vector2D=<20 20>, Spawn Group Mask=1)
     Spawn Group Mask=0
    )
    Sprite Size Mode=::NewEnumerator4
    UsePositionOffset=false
    Write Material Random=true
   )]
  particleUpdateScript [
   &SolveForcesAndVelocity(Clamp Velocity=false, Limit Acceleration=false)
   &ParticleState(Kill Particles When Lifetime Has Elapsed=false)
   &ScaleSpriteSize(
    Scale Sprite Size Mode=::NewEnumerator0
    Uniform Scale Factor=&MakeCustomFloatFromBool(
     Bool=$User.bisHandHidden
     True Float=&MaskFloatBySpawnGroup(Masked Float=0, Passthrough Float=1)
     False Float=1
    )
   ) {module=ScaleSpriteSize001}
   &ScaleSpriteSize(
    Uniform Curve Index=$User.ReactiveCurve
    Uniform Curve Sprite Scale=DataInterface() {propertyValues {Curve {keys [{tangentMode=RCTM_User, time=0, value=0}, <at 0.0062 1>, <at 0.9916 1>, {tangentMode=RCTM_User, time=1, value=0}]}}}
   )
   &SocketLocation(
    SamplePosOffset=true
    Sampled Position Offset=&MaskVectorBySpawnGroup(
     Masked Vector=<0.4022 0.0572 -3.4576>
     Passthrough Vector=&MaskVectorBySpawnGroup(Masked Vector=<-1.301 -0.6179 -2.5435>, Passthrough Vector=<0.13 -0.1039 -0.43>, Spawn Group Mask=1)
    )
    Socket Index=&MaskIntBySpawnGroup(
     Masked Int=2
     Passthrough Int=/Niagara/DynamicInputs/Helpers/MaskInt(Mask Source:NiagaraInt32=0, Mask Parameter:NiagaraInt32=?, Mask:NiagaraInt32=1, Masked Int:NiagaraInt32=1, Passthrough Int:NiagaraInt32=0)
     Spawn Group Mask=0
    )
    Socket Reader=$System.Glow_Ice_Sockets
    Socket Sampling Mode=::NewEnumerator3
    VelocitySpeedLimit=false
   )
   &AlignSpriteToMeshOrientation(
    Mesh Orientation Relative Sprite Facing Vector:Vector3f=&MaskVectorBySpawnGroup(
     Masked Vector=<1 0 4.3208>
     Passthrough Vector=&MaskVectorBySpawnGroup(Masked Vector=<1 1 1.8897>, Passthrough Vector=<-1.5616 -8.6923 1.8086>, Spawn Group Mask=1)
    )
    SetAlignment=true
    SetFacing=true
   )
   &ScaleColor(
    Scale Alpha=&FloatFromCurve(
     FloatCurve=DataInterface() {propertyValues {Curve {keys [
         <at 0 0 RCIM_Cubic>
         {arriveTangentWeight=0.0454, interpMode=RCIM_Cubic, leaveTangentWeight=0.3049, tangentMode=RCTM_User, time=0.0265, value=1}
         {arriveTangentWeight=0.2121, interpMode=RCIM_Cubic, leaveTangentWeight=0.2593, tangentMode=RCTM_User, time=0.3227, value=1}
         {arriveTangent=-5.937, interpMode=RCIM_Cubic, leaveTangent=-5.937, tangentMode=RCTM_User, time=0.7219, value=0}
        ]}}}
     CurveIndex=$User.ReactiveCurve
    )
    Scale RGB=&MakeVectorFromLinearColorRGB(Color=&ColorFromCurve(
      ColorCurve=DataInterface() {propertyValues {
        AlphaCurve {keys [<at 0 1 RCIM_Cubic>]}
        BlueCurve {keys [<at 0 1 RCIM_Cubic>,<at 0.0495 0>,<at 0.5588 0>,<at 0.8324 1 RCIM_Cubic>]}
        GreenCurve {keys [<at 0 0.8963 RCIM_Cubic>,<at 0.0495 0>,<at 0.5588 0>,<at 0.8324 0.8963 RCIM_Cubic>]}
        RedCurve {keys [<at 0 0.3231 RCIM_Cubic>,<at 0.0495 1>,<at 0.5588 1>,<at 0.8324 0.3231 RCIM_Cubic>]}
       }}
      CurveIndex=$User.ReactiveCurve
     ))
    ScaleA=true
    ScaleRGB=true
   ) {module=ScaleColor001}
  ]
  rendererClasses [&NiagaraSpriteRendererProperties]
  renderers [{rendererClass=&NiagaraSpriteRendererProperties}]
 }
 Glow_SideView {
  emitterUpdateScript [
   &EmitterState(Life Cycle Mode=::NewEnumerator1, Loop Behavior=::NewEnumerator1, Loop Duration=1, UseLoopDelay=false)
   &SpawnBurst_Instantaneous(Spawn Group=2, Use Spawn Probability=false, UseLoopCountLimit=false) {enabled=false, module=SpawnBurst_Instantaneous002}
   &SpawnBurst_Instantaneous(Spawn Group=1, Use Spawn Probability=false, UseLoopCountLimit=false)
   &SpawnBurst_Instantaneous(Use Spawn Probability=false, UseLoopCountLimit=false) {module=SpawnBurst_Instantaneous001}
  ]
  particleSpawnScript [
   /SetVariables(Particles.VisibilityTag=$Emitter.SpawnGroup) {isSetParameters=true}
   /SetVariables(Particles.GlowPos:NiagaraInt32=&RandomRangeInt(Minimum=0, Maximum=1)) {isSetParameters=true}
   &InitializeParticle(
    Color=&MakeCustomLinearColorFromBool(Bool=$User.bIsFrontEndPreview, True Color=<rgba 1 1 1 0.25>, False Color=<rgba 1 1 1 0.5>)
    Lifetime=5
    Mesh Scale Mode=::NewEnumerator3
    UsePositionOffset=false
    Write Material Random=true
   )
   &InitialMeshOrientation(
    Mesh Orientation Mode=::NewEnumerator2
    Rotation=&MaskVectorBySpawnGroup(Masked Vector=<0 0.5183 0.0548>, Passthrough Vector=<0 0.5752 -0.0913>)
    Use Rotation Vector=true
   )
  ]
  particleUpdateScript [
   &ParticleState(Kill Particles When Lifetime Has Elapsed=false)
   &ScaleMeshSize(Scale Factor=&Multiply_VectorByFloat(
     Vector=<1 1 1>
     Float=&MakeCustomFloatFromBool(
      Bool=$User.bisHandHidden
      True Float=&MaskFloatBySpawnGroup(Masked Float=0, Passthrough Float=1)
      False Float=1
     )
    )) {module=ScaleMeshSize001}
   &ScaleMeshSize(Scale Factor=&MakeVector(
     X=1
     Y=1
     Z=&FloatFromCurve(
      FloatCurve=DataInterface() {propertyValues {Curve {keys [
          <at 0 0 RCIM_Cubic>
          {arriveTangentWeight=0.0224, interpMode=RCIM_Cubic, leaveTangentWeight=0.4248, tangentMode=RCTM_User, time=0.0208, value=1.5}
          {arriveTangentWeight=0.0393, interpMode=RCIM_Cubic, leaveTangentWeight=0.4205, tangentMode=RCTM_User, time=0.0533, value=1}
          <at 1 1 RCIM_Cubic>
         ]}}}
      CurveIndex=$User.ReactiveCurve
     )
    ))
   &SocketLocation(
    SamplePosOffset=true
    Sampled Position Offset=&MaskVectorBySpawnGroup(Masked Vector=<0.4964 0.5531 0.8095>, Passthrough Vector=<-0.1637 0.1049 -0.2231>)
    Socket Index=&MaskIntBySpawnGroup(Masked Int=2, Passthrough Int=1, Spawn Group Mask=0)
    Socket Reader=$System.Glow_Ice_Sockets
    Socket Sampling Mode=::NewEnumerator3
    VelocitySpeedLimit=false
   )
   &ScaleColor(
    Scale Alpha=&FloatFromCurve(
     FloatCurve=DataInterface() {propertyValues {Curve {keys [
         <at 0 1 RCIM_Cubic>
         {arriveTangentWeight=0.2121, interpMode=RCIM_Cubic, leaveTangentWeight=0.2593, tangentMode=RCTM_User, time=0.3227, value=1}
         {arriveTangent=-5.937, interpMode=RCIM_Cubic, leaveTangent=-5.937, tangentMode=RCTM_User, time=0.7219, value=0}
        ]}}}
     CurveIndex=$User.ReactiveCurve
    )
    Scale RGB=&MakeVectorFromLinearColorRGB(Color=&ColorFromCurve(
      ColorCurve=DataInterface() {propertyValues {
        AlphaCurve {keys [<at 0 1 RCIM_Cubic>]}
        BlueCurve {keys [<at 0 1 RCIM_Cubic>,<at 0.0495 0>,<at 0.5588 0>,<at 0.8324 1 RCIM_Cubic>]}
        GreenCurve {keys [<at 0 0.8963 RCIM_Cubic>,<at 0.0495 0>,<at 0.5588 0>,<at 0.8324 0.8963 RCIM_Cubic>]}
        RedCurve {keys [<at 0 0.3231 RCIM_Cubic>,<at 0.0495 1>,<at 0.5588 1>,<at 0.8324 0.3231 RCIM_Cubic>]}
       }}
      CurveIndex=$User.ReactiveCurve
     ))
    ScaleA=true
    ScaleRGB=true
   ) {module=ScaleColor001}
  ]
  rendererClasses [&NiagaraMeshRendererProperties]
  renderers [
   {rendererClass=&NiagaraMeshRendererProperties, rendererIndex=0}
   {rendererClass=&NiagaraMeshRendererProperties, rendererIndex=1}
  ]
 }
 Heat_Distortion {
  emitterUpdateScript [
   &EmitterState()
   &SpawnRate(
    SpawnRate=&Multiply_Float(
     A=20
     B=&FloatFromCurve(
      FloatCurve=DataInterface() {propertyValues {Curve {keys [
          <at 0 0 RCIM_Cubic>
          {arriveTangentWeight=0.042, interpMode=RCIM_Cubic, leaveTangentWeight=0.3338, tangentMode=RCTM_User, time=0.0582, value=1}
          {arriveTangentWeight=0.2646, interpMode=RCIM_Cubic, leaveTangentWeight=0.1769, tangentMode=RCTM_User, time=0.4451, value=1}
          {arriveTangent=-7.9875, interpMode=RCIM_Cubic, leaveTangent=-7.9875, tangentMode=RCTM_User, time=0.7932, value=0}
         ]}}}
      CurveIndex=$User.ReactiveCurve
     )
    )
    Use Spawn Probability=false
   )
  ]
  particleSpawnScript [
   &InitializeParticle(
    Lifetime=&RandomRangeFloat(Minimum=0.25, Maximum=0.5)
    Sprite Size Mode=::NewEnumerator3
    Uniform Sprite Size=40
    UsePositionOffset=false
    Write Material Random=true
   )
   /SetVariables(Particles.SocketIndex:NiagaraInt32=&RandomRangeInt(Minimum=7, Maximum=13)) {isSetParameters=true}
   &ApplyInitialForces()
  ]
  particleUpdateScript [
   &ParticleState()
   &ScaleSpriteSize(
    Scale Sprite Size Mode=::NewEnumerator0
    Uniform Scale Factor=&Lerp_Float(
     A=&SelectFloatFromArray(Direct Array Index=$Particles.SocketIndex, Float Selection Array=DataInterface() {propertyValues {FloatData [0,1,0,0,0,0,0,1,0,1,0,1,1,1]}})
     B=1
     Alpha=&MakeCustomFloatFromBool(Bool=$User.bisHandHidden, True Float=0, False Float=1)
    )
   )
   &SocketLocation(
    SamplePosOffset=true
    Sampled Position Offset=&SelectVectorFromArray(
     Direct Array Index=$Particles.SocketIndex
     Vector Selection Array=DataInterface() {propertyValues {FloatData [<0 0 0>, <0 0 0>, <0 0 0>, <0 0 0>, <0 0 0>, <0 0 0>, <0 0 0>, <0 0 0>, <0 0 0>, <0 0 0>, <0 0 0>, <0 0 0>, <0 0 0>, <0 -10 0>]}}
    )
    Socket Index=$Particles.SocketIndex
    Socket Reader=$System.Glow_Ice_Sockets
    Socket Sampling Mode=::NewEnumerator3
    VelocitySpeedLimit=false
   )
   &ShapeLocation(MaskSpawnGroup=false, Override Seed=false, Shape Origin=$Output.SocketLocation.SampledPosition, Sphere Radius=5, UseOwnerScale=true)
   &DynamicMaterialParameters(
    Index 0 Param 1=&FloatFromCurve(FloatCurve=DataInterface() {propertyValues {Curve {keys [<at 0 1 RCIM_Cubic>, <at 0.5 0.983 RCIM_Cubic>, {interpMode=RCIM_Cubic, leaveTangent=-0.0639, time=1, value=1}]}}})
    Index 0 Param 2=0.5
    Module.Index0Float0Write=true
    Module.Index0Float1Write=false
    Module.Index0Float2Write=false
    Module.Index0Float3Write=false
   )
   &OffsetPosition(Position Offset=&MakeVector(
     X=0
     Y=0
     Z=&FloatFromCurve(
      FloatCurve=DataInterface() {propertyValues {Curve {defaultValue=1, keys [<at 0 0 RCIM_Cubic>, {tangentMode=RCTM_User, time=1, value=1}]}}}
      Scale Curve=5
     )
    )) {enabled=false}
   &CameraOffset(Camera Offset Amount=&FN_FrontEndFloatModifier(
     Gameplay=&SelectFloatFromArray(
      Direct Array Index=$Particles.SocketIndex
      Float Selection Array=DataInterface() {propertyValues {FloatData [0,0,0,0,0,10,10,10,20,10,20,20,20,10]}}
     )
     Front End=10
    ))
  ]
  rendererClasses [&NiagaraSpriteRendererProperties]
  renderers [
   {rendererClass=&NiagaraSpriteRendererProperties}
   {rendererClass=&NiagaraSpriteRendererProperties, rendererIndex=1}
  ]
 }
 Heat_Smoke {
  emitterUpdateScript [
   &EmitterState(Life Cycle Mode=::NewEnumerator1, Loop Behavior=::NewEnumerator1, Loop Duration=1, UseLoopDelay=false)
   &SpawnBurst_Instantaneous(Spawn Count=11, Use Spawn Probability=false, UseLoopCountLimit=false)
  ]
  particleSpawnScript [&InitializeParticle(
    Color=<rgba 1 0.0558 0 0.5>
    Lifetime=3
    Sprite Rotation Mode=::NewEnumerator2
    Sprite Size Mode=::NewEnumerator3
    Uniform Sprite Size=30
    UsePositionOffset=false
    Write Material Random=true
   )]
  particleUpdateScript [
   &ParticleState(Kill Particles When Lifetime Has Elapsed=false, Loop Particles Lifetime=true)
   &SubUVAnimation(
    "Sprite Renderer "=DataInterface() {propertyValues {SpriteRenderer=&self:Heat_Smoke_0.NiagaraSpriteRendererProperties_0}}
    Start Frame Offset:NiagaraInt32=&RandomRangeInt(Minimum=0, Maximum=63)
    UseEndFrame=false
    UseStartFrame=false
   )
   &SocketLocation(
    SamplePosOffset=true
    Sampled Position Offset=&SelectVectorFromArray(
     Direct Array Index=&ReturnExecIndex()
     Vector Selection Array=DataInterface() {propertyValues {FloatData [<0 0 0>, <-3 0 -2>, <0 0 -2>, <3.5896 -9.2127 0>, <-18 -9 3>, <20 9 -4>, <-17 9 3>, <17 -9 -3>, <2 -12 0>, <2 -3 -12>, <-2 10 -8>]}}
    )
    Socket Index=&SelectIntFromArray(
     Direct Array Index=&ReturnExecIndex()
     Int Selection Array=DataInterface() {propertyValues {IntData [0,1,2,13,11,12,11,12,14,15,15]}}
    )
    Socket Reader=$System.Glow_Ice_Sockets
    Socket Sampling Mode=::NewEnumerator3
    VelocitySpeedLimit=false
   )
   &ScaleSpriteSize(
    Scale Sprite Size Mode=::NewEnumerator0
    Uniform Scale Factor=&Lerp_Float(
     A=&SelectFloatFromArray(
      Direct Array Index=&ReturnExecIndex()
      Float Selection Array=DataInterface() {propertyValues {FloatData [1,1,0,1,1,1,1,1,1,1,1]}}
     )
     B=1
     Alpha=&MakeCustomFloatFromBool(Bool=$User.bisHandHidden, True Float=0, False Float=1)
    )
   ) {module=ScaleSpriteSize001}
   &ScaleColor(
    Scale Alpha=&FloatFromCurve(
     FloatCurve=DataInterface() {propertyValues {Curve {keys [
         <at 0 0 RCIM_Cubic>
         {arriveTangentWeight=0.0748, interpMode=RCIM_Cubic, leaveTangentWeight=0.2754, tangentMode=RCTM_User, time=0.1439, value=1.0029}
         {arriveTangentWeight=0.2121, interpMode=RCIM_Cubic, leaveTangentWeight=0.2593, tangentMode=RCTM_User, time=0.3227, value=1}
         {arriveTangent=-5.937, interpMode=RCIM_Cubic, leaveTangent=-5.937, tangentMode=RCTM_User, time=0.7219, value=0}
        ]}}}
     CurveIndex=$User.ReactiveCurve
    )
    Scale RGB=&MakeVectorFromLinearColorRGB(Color=&ColorFromCurve(
      ColorCurve=DataInterface() {propertyValues {
        AlphaCurve {keys [<at 0 1 RCIM_Cubic>]}
        BlueCurve {keys [<at 0 1 RCIM_Cubic>,<at 0.0495 0>,<at 0.5588 0>,<at 0.8324 1 RCIM_Cubic>]}
        GreenCurve {keys [<at 0 0.8963 RCIM_Cubic>,<at 0.0495 0>,<at 0.5588 0>,<at 0.8324 0.8963 RCIM_Cubic>]}
        RedCurve {keys [<at 0 0.3231 RCIM_Cubic>,<at 0.0495 1>,<at 0.5588 1>,<at 0.8324 0.3231 RCIM_Cubic>]}
       }}
      CurveIndex=$User.ReactiveCurve
     ))
    ScaleA=true
    ScaleRGB=true
   ) {module=ScaleColor001}
   &CameraOffset(Camera Offset Amount=3)
   &ScaleSpriteSize(
    Uniform Curve Index=$User.ReactiveCurve
    Uniform Curve Sprite Scale=DataInterface() {propertyValues {Curve {keys [{tangentMode=RCTM_User, time=0, value=0}, <at 0.0062 1>, <at 0.9916 1>, {tangentMode=RCTM_User, time=1, value=0}]}}}
   )
  ]
  rendererClasses [&NiagaraSpriteRendererProperties]
  renderers [{rendererClass=&NiagaraSpriteRendererProperties}]
 }
 IceShards {
  emitterUpdateScript [
   &EmitterState()
   &SpawnPerFrame(
    Spawn:NiagaraBool=&SetBoolByFloatComparison(A=0, B=1, Comparison Type=::NewEnumerator4)
    Spawn Count=20
    Use Spawn Probability=false
   ) {enabled=false}
   &SpawnPerFrame(
    Spawn:NiagaraBool=&Boolean_AndOperation(
     Bool1=&Boolean_AndOperation(
      Bool1=&Boolean_NotOperation(Bool=$System.UpdateValueStateChange.Changed)
      Bool2=&Boolean_NotOperation(Bool=$Transient.FirstFrame)
     )
     Bool2=true
    )
    Spawn Count=20
    Use Spawn Probability=false
   ) {module=SpawnPerFrame001}
   &SpawnBurst_Instantaneous(Spawn Count=20, Use Spawn Probability=false, UseLoopCountLimit=false) {enabled=false}
  ]
  particleSpawnScript [
   /SetVariables(Particles.MeshIndex=&RandomRangeInt(Minimum=0, Maximum=1)) {isSetParameters=true, module=SetVariables_#3}
   /SetVariables(Particles.SocketIndex:NiagaraInt32=&SelectIntFromArray(
     Array Sampling Mode=::NewEnumerator1
     Direct Array Index=?
     Int Selection Array=DataInterface() {propertyValues {IntData [1,2,3,4,5,6]}}
     Random Seed=0
    )) {isSetParameters=true, module=SetVariables_#4}
   &InitializeParticle(
    Lifetime=&RandomRangeFloat(Minimum=0.8, Maximum=1.2)
    Mesh Scale=&Multiply_VectorByFloat(Vector=<1 1 1>, Float=&RandomRangeFloat(Minimum=1, Maximum=1.5))
    Mesh Scale Mode=::NewEnumerator4
    Sprite Size Mode=::NewEnumerator3
    Uniform Sprite Size=1
    UsePositionOffset=false
    Write Material Random=true
   )
   &SocketLocation(
    Orientation Sampling=::NewEnumerator1
    SamplePosOffset=false
    Socket Index=$Particles.SocketIndex
    Socket Reader=$System.Glow_Ice_Sockets
    Socket Sampling Mode=::NewEnumerator3
    VelocitySpeedLimit=false
   )
   &ShapeLocation(
    Additional Quaternion Rotation=true
    Additional Yaw / Pitch / Roll=&SelectVectorFromArray(
     Direct Array Index=$Particles.SocketIndex
     Vector Selection Array=DataInterface() {propertyValues {FloatData [<0 0 0>, <0 -25 90>, <0 -15 90>, <0 -15 -90>, <0 -15 -90>, <0 0 79.1764>, <0 0 -90>]}}
    )
    Hemisphere Distribution=<0.5 1>
    MaskSpawnGroup=false
    Non Uniform Scale=&SelectVectorFromArray(
     Direct Array Index=$Particles.SocketIndex
     Vector Selection Array=DataInterface() {propertyValues {FloatData [<1 1 1>, <1 0.5 0.7095>, <1.5 0.5 0.5303>, <1 0.5 0.8125>, <1.5 0.5 0.5003>, <1 0.5 0.8209>, <1.3 0.5 0.8209>]}}
    )
    Offset=&SelectVectorFromArray(
     Direct Array Index=$Particles.SocketIndex
     Vector Selection Array=DataInterface() {propertyValues {FloatData [<0 0 0>, <3.9501 -0.9711 0>, <3.0263 0 0>, <-5.2039 0.9974 0>, <-2.3036 0 0>, <-4.0002 0 -0.264>, <3.1821 0 -2.4683>]}}
    )
    Offset Mode=::NewEnumerator0
    Override Seed=false
    Rotation Angle Type:ENiagara_AngleInput=::NewEnumerator0
    Rotation Mode=::NewEnumerator3
    Rotation Quaternion=$Output.SocketLocation.SampledRotation
    Shape Origin=$Output.SocketLocation.SampledPosition
    Sphere Radius=10
    Sphere Surface Distribution=1
    UseOwnerScale=true
   )
   &InitialMeshOrientation(Use Rotation Vector=false)
   &AddVelocity(Velocity=&Multiply_VectorByFloat(Vector=$Particles.ShapeLocation.ShapeNormal, Float=&RandomRangeFloat(Minimum=150, Maximum=250)))
   &ApplyInitialForces()
  ]
  particleUpdateScript [
   &ParticleState()
   &ScaleMeshSize(Scale Factor=&Multiply_VectorByFloat(
     Vector=<1 1 1>
     Float=&Lerp_Float(
      A=1
      B=&MergeHelper_FloatSwitch(0=1, 1=1, 2=0, 3=1, 4=0, 5=1, Index=$Particles.SocketIndex, Num=::NewEnumerator4)
      Alpha=&MakeCustomFloatFromBool(Bool=$User.bisHandHidden, True Float=1, False Float=0)
     )
    )) {module=ScaleMeshSize001}
   &UpdateMeshOrientation(Rotation Rate=90, Rotation Vector=<0 1 1>)
   &GravityForce()
   &SolveForcesAndVelocity(Clamp Velocity=false, Limit Acceleration=false)
  ]
  rendererClasses [&NiagaraMeshRendererProperties]
  renderers [{rendererClass=&NiagaraMeshRendererProperties, rendererIndex=0}]
 }
 IceShards001 {
  emitterUpdateScript [
   &EmitterState()
   &SpawnPerFrame(
    Spawn:NiagaraBool=&SetBoolByFloatComparison(A=0, B=1, Comparison Type=::NewEnumerator4)
    Spawn Count=20
    Use Spawn Probability=false
   ) {enabled=false}
   &SpawnPerFrame(
    Spawn:NiagaraBool=&Boolean_AndOperation(
     Bool1=&Boolean_AndOperation(
      Bool1=&Boolean_NotOperation(Bool=$System.UpdateValueStateChange.Changed)
      Bool2=&Boolean_NotOperation(Bool=$Transient.FirstFrame)
     )
     Bool2=true
    )
    Spawn Count=15
    Use Spawn Probability=false
   ) {module=SpawnPerFrame001}
   &SpawnBurst_Instantaneous(Spawn Count=20, Use Spawn Probability=false, UseLoopCountLimit=false) {enabled=false}
  ]
  particleSpawnScript [
   /SetVariables(Particles.MeshIndex=&RandomRangeInt(Minimum=0, Maximum=1)) {isSetParameters=true, module=SetVariables_#3}
   /SetVariables(Particles.SocketIndex:NiagaraInt32=&SelectIntFromArray(
     Array Sampling Mode=::NewEnumerator1
     Direct Array Index=?
     Int Selection Array=DataInterface() {propertyValues {IntData [1,3]}}
     Random Seed=0
    )) {isSetParameters=true, module=SetVariables_#4}
   &InitializeParticle(
    Lifetime=&RandomRangeFloat(Minimum=0.8, Maximum=1.2)
    Mesh Scale=&Multiply_VectorByFloat(Vector=<1 1 1>, Float=&RandomRangeFloat(Minimum=1, Maximum=1.5))
    Mesh Scale Mode=::NewEnumerator4
    Sprite Size Mode=::NewEnumerator3
    Uniform Sprite Size=1
    UsePositionOffset=false
    Write Material Random=true
   )
   &SocketLocation(
    Orientation Sampling=::NewEnumerator1
    SamplePosOffset=false
    Socket Index=$Particles.SocketIndex
    Socket Reader=$System.Glow_Ice_Sockets
    Socket Sampling Mode=::NewEnumerator3
    VelocitySpeedLimit=false
   )
   &ShapeLocation(
    Additional Quaternion Rotation=true
    Additional Yaw / Pitch / Roll=&SelectVectorFromArray(
     Direct Array Index=$Particles.SocketIndex
     Vector Selection Array=DataInterface() {propertyValues {FloatData [<0 0 0>, <0 -25 90>, <0 -15 90>, <0 -15 -90>, <0 -15 -90>, <0 0 79.1764>, <0 0 -90>]}}
    )
    Hemisphere Distribution=<0.5 1>
    MaskSpawnGroup=false
    Non Uniform Scale=&SelectVectorFromArray(
     Direct Array Index=$Particles.SocketIndex
     Vector Selection Array=DataInterface() {propertyValues {FloatData [<1 1 1>, <1 0.5 0.7095>, <1.5 0.5 0.5303>, <1 0.5 0.8125>, <1.5 0.5 0.5003>, <1 0.5 0.8209>, <1.3 0.5 0.8209>]}}
    )
    Offset=&SelectVectorFromArray(
     Direct Array Index=$Particles.SocketIndex
     Vector Selection Array=DataInterface() {propertyValues {FloatData [<0 0 0>, <3.9501 -0.9711 0>, <3.0263 0 0>, <-5.2039 0.9974 0>, <-2.3036 0 0>, <-4.0002 0 -0.264>, <3.1821 0 -2.4683>]}}
    )
    Offset Mode=::NewEnumerator0
    Override Seed=false
    Rotation Angle Type:ENiagara_AngleInput=::NewEnumerator0
    Rotation Mode=::NewEnumerator3
    Rotation Quaternion=$Output.SocketLocation.SampledRotation
    Shape Origin=$Output.SocketLocation.SampledPosition
    Sphere Radius=10
    Sphere Surface Distribution=1
    UseOwnerScale=true
   )
   &InitialMeshOrientation(Use Rotation Vector=false)
   &AddVelocity(Velocity=&Multiply_VectorByFloat(Vector=$Particles.ShapeLocation.ShapeNormal, Float=&RandomRangeFloat(Minimum=100, Maximum=150)))
   &ApplyInitialForces()
   &AddVelocity(Velocity=&MakeVector(X=0, Y=0, Z=&RandomRangeFloat(Minimum=200, Maximum=300))) {module=AddVelocity001}
  ]
  particleUpdateScript [
   &ParticleState()
   &UpdateMeshOrientation(Rotation Rate=90, Rotation Vector=<0 1 1>)
   &GravityForce()
   &SolveForcesAndVelocity(Clamp Velocity=false, Limit Acceleration=false)
  ]
  rendererClasses [&NiagaraMeshRendererProperties]
  renderers [{rendererClass=&NiagaraMeshRendererProperties, rendererIndex=0}]
 }
 IceSmoke {
  emitterUpdateScript [
   &EmitterState()
   &SpawnBurst_Instantaneous(Spawn Count=20, Use Spawn Probability=false, UseLoopCountLimit=false) {enabled=false}
   &SpawnPerFrame(
    Spawn:NiagaraBool=&Boolean_AndOperation(
     Bool1=&Boolean_AndOperation(
      Bool1=&Boolean_NotOperation(Bool=$System.UpdateValueStateChange.Changed)
      Bool2=&Boolean_NotOperation(Bool=$Transient.FirstFrame)
     )
     Bool2=true
    )
    Spawn Count=30
    Use Spawn Probability=false
   ) {module=SpawnPerFrame001}
   &SpawnPerFrame(
    Spawn:NiagaraBool=&SetBoolByFloatComparison(A=0, B=1, Comparison Type=::NewEnumerator4)
    Spawn Count=10
    Use Spawn Probability=false
   ) {enabled=false}
  ]
  particleSpawnScript [
   /SetVariables(Particles.SocketIndex:NiagaraInt32=&RandomRangeInt(Minimum=1, Maximum=6)) {isSetParameters=true}
   &InitializeParticle(
    Color=<rgba 0.5885 0.9226 1 0.15>
    Lifetime=&RandomRangeFloat(Minimum=0.5, Maximum=1)
    Sprite Rotation Mode=::NewEnumerator2
    Sprite Size Mode=::NewEnumerator3
    Uniform Sprite Size=&RandomRangeFloat(Minimum=30, Maximum=60)
    UsePositionOffset=false
    Write Material Random=true
   )
   &SocketLocation(
    Orientation Sampling=::NewEnumerator1
    SamplePosOffset=false
    Socket Index=$Particles.SocketIndex
    Socket Reader=$System.Glow_Ice_Sockets
    Socket Sampling Mode=::NewEnumerator3
    VelocitySpeedLimit=false
   )
   &ShapeLocation(MaskSpawnGroup=false, Override Seed=false, Shape Origin=$Output.SocketLocation.SampledPosition, Sphere Radius=10, UseOwnerScale=true)
  ]
  particleUpdateScript [
   &ParticleState()
   &ScaleSpriteSize(
    Scale Sprite Size Mode=::NewEnumerator0
    Uniform Scale Factor=&Lerp_Float(
     A=&MergeHelper_FloatSwitch(0=1, 1=1, 2=0, 3=1, 4=0, 5=1, Index=$Particles.SocketIndex, Num=::NewEnumerator4)
     B=1
     Alpha=&MakeCustomFloatFromBool(Bool=$User.bisHandHidden, True Float=0, False Float=1)
    )
   ) {module=ScaleSpriteSize001}
   &ScaleSpriteSize(Uniform Curve Sprite Scale=DataInterface() {propertyValues {Curve {keys [
        {arriveTangent=2.478, interpMode=RCIM_Cubic, leaveTangent=2.478, tangentMode=RCTM_User, time=0, value=0.25}
        {arriveTangent=0.312, interpMode=RCIM_Cubic, leaveTangent=0.312, tangentMode=RCTM_User, time=1.0024, value=1}
       ]}}})
   &SubUVAnimation(
    "Sprite Renderer "=DataInterface() {propertyValues {SpriteRenderer=&self:IceSmoke_0.NiagaraSpriteRendererProperties_0}}
    UseEndFrame=false
    UseStartFrame=false
   )
   &ScaleColor(
    Scale Alpha=&FloatFromCurve(FloatCurve=DataInterface() {propertyValues {Curve {
        defaultValue=1
        keys [{interpMode=RCIM_Cubic, tangentMode=RCTM_User, time=0, value=1}, {interpMode=RCIM_Cubic, tangentMode=RCTM_User, time=1, value=0}]
       }}})
    ScaleA=true
    ScaleRGB=true
   )
  ]
  rendererClasses [&NiagaraSpriteRendererProperties]
  renderers [{rendererClass=&NiagaraSpriteRendererProperties}]
 }
 Light {
  emitterUpdateScript [
   &EmitterState(Life Cycle Mode=::NewEnumerator1, Loop Behavior=::NewEnumerator1, Loop Duration=1, UseLoopDelay=false)
   &SpawnBurst_Instantaneous(Use Spawn Probability=false, UseLoopCountLimit=false) {module=SpawnBurst_Instantaneous001}
   &SpawnBurst_Instantaneous(Spawn Group=1, Use Spawn Probability=false, UseLoopCountLimit=false)
   &SpawnBurst_Instantaneous(Spawn Group=2, Use Spawn Probability=false, UseLoopCountLimit=false) {module=SpawnBurst_Instantaneous002}
  ]
  particleSpawnScript [&InitializeParticle(Lifetime=5, UsePositionOffset=false, Write Material Random=true)]
  particleUpdateScript [
   &ParticleState(Kill Particles When Lifetime Has Elapsed=false)
   &SocketLocation(
    SamplePosOffset=true
    Sampled Position Offset=&MaskVectorBySpawnGroup(
     Masked Vector=<0.4022 0.0572 -3.4576>
     Passthrough Vector=&MaskVectorBySpawnGroup(Masked Vector=<-1.301 -0.6179 -2.5435>, Passthrough Vector=<0.13 -0.1039 -0.43>, Spawn Group Mask=1)
    )
    Socket Index=&MaskIntBySpawnGroup(
     Masked Int=2
     Passthrough Int=&MaskInt(Mask Source:NiagaraInt32=0, Mask Parameter:NiagaraInt32=?, Mask:NiagaraInt32=1, Masked Int:NiagaraInt32=1, Passthrough Int:NiagaraInt32=0)
     Spawn Group Mask=0
    )
    Socket Reader=$System.Glow_Ice_Sockets
    Socket Sampling Mode=::NewEnumerator3
    VelocitySpeedLimit=false
   )
   /SetVariables(Particles.LightInten:NiagaraFloat=&Lerp_Float(
     A=&MaskFloatBySpawnGroup(Masked Float=0, Passthrough Float=10)
     B=1
     Alpha=&MakeCustomFloatFromBool(Bool=$User.bisHandHidden, True Float=0, False Float=1)
    )) {isSetParameters=true}
   &FN_Light_Attributes(
    Color=<rgba 1 0 0 1>
    Create Color Attribute=true
    Create Exponent Attribute=true
    Create Radius Attribute=true
    Enable Particle Light Enable=true
    Exponent=5
    Light Enabled:NiagaraBool=&SetBoolByFloatComparison(
     A=&FloatFromCurve(FloatCurve=DataInterface() {propertyValues {Curve {keys [<at 0 0>,<at 0.0272 1>,<at 0.9625 1>,<at 1 0>]}}}, CurveIndex=$User.ReactiveCurve)
     B=0.1
    )
    Radius=50
    Scale Color=&FloatFromCurve(
     FloatCurve=DataInterface() {propertyValues {Curve {keys [
         <at 0 0 RCIM_Cubic>
         {arriveTangentWeight=0.0454, interpMode=RCIM_Cubic, leaveTangentWeight=0.3049, tangentMode=RCTM_User, time=0.0265, value=1}
         {arriveTangentWeight=0.2121, interpMode=RCIM_Cubic, leaveTangentWeight=0.2593, tangentMode=RCTM_User, time=0.3227, value=1}
         {arriveTangent=-5.937, interpMode=RCIM_Cubic, leaveTangent=-5.937, tangentMode=RCTM_User, time=0.7219, value=0}
        ]}}}
     CurveIndex=$User.ReactiveCurve
     Scale Curve=$Particles.LightInten
    )
   )
  ]
  rendererClasses [&NiagaraLightRendererProperties]
  renderers [{rendererClass=&NiagaraLightRendererProperties, rendererIndex=0}]
 }
}
summary {
 systemName=NS_Character_FoundKelp_Bolt_Castfall_Idle
 emitters [
  {emitterName=Heat_Distortion, rendererClasses [&NiagaraSpriteRendererProperties]}
  {emitterName=IceShards, rendererClasses [&NiagaraMeshRendererProperties]}
  {emitterName=IceSmoke, rendererClasses [&NiagaraSpriteRendererProperties]}
  {emitterName=Glow_FrontView, rendererClasses [&NiagaraSpriteRendererProperties]}
  {emitterName=Glow_SideView, rendererClasses [&NiagaraMeshRendererProperties]}
  {emitterName=Heat_Smoke, rendererClasses [&NiagaraSpriteRendererProperties]}
  {emitterName=Light, rendererClasses [&NiagaraLightRendererProperties]}
  {emitterName=IceShards001, rendererClasses [&NiagaraMeshRendererProperties]}
  {emitterName=FlyingShards, rendererClasses [&NiagaraMeshRendererProperties]}
 ]
 userVariables [
  User.bisFootHidden {type=NiagaraBool}
  User.bisFrontEnd {type=NiagaraBool}
  User.bIsFrontEndPreview {type=NiagaraBool}
  User.bisHandHidden {type=NiagaraBool}
  User.KillCount {type=NiagaraInt32}
  User.ReactiveCurve {type=NiagaraFloat, defaultValue=0, description=""}
 ]
}
