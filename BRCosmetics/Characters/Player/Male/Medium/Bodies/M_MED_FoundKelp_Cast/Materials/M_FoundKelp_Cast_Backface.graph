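# Masked, two-sided skeletal mesh material for the FoundKelp Cast character body that renders only back faces, filled with flat tinted base and emissive colors.
# An inverted, saturated TwoSidedSign discards front faces, and the FKC_IceBackfaceCullingMask texture's red channel further masks which back faces show.
# The resulting material attributes pass through MF_TechArtOverrides, then the shared BaseFN_CharacterOpacity_2023 character opacity function.
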Material /BRCosmetics/Characters/Player/Male/Medium/Bodies/M_MED_FoundKelp_Cast/Materials/M_FoundKelp_Cast_Backface
graph {
 properties {
  bAutomaticallySetUsageInEditor=false
  bUseMaterialAttributes=true
  bUsedWithMorphTargets=true
  bUsedWithSkeletalMesh=true
  blendMode=BLEND_Masked
  twoSided=true
  usedShadingModels=Default Lit
 }
 nodes [
  n1 = VectorParameter() {parameterName=Basecolor, defaultValue=<rgba 0.161 0.203 0.224 1>}
  n2 = TwoSidedSign()
  n3 = Multiply(
   A=Saturate(None=OneMinus(None=@n2))
   B=TextureSampleParameter2D() {parameterName=FKC_IceBackfaceCullingMask, texture=&T_WhitesRGB}.R
  )
  n4 = VectorParameter() {parameterName=Emissive, defaultValue=<rgba 0.1582 0.2239 0.3177 1>}
  MaterialFunctionCall!(Material=MaterialFunctionCall(In=SetMaterialAttributes!(MaterialAttributes, Base Color=@n1.RGB, Opacity Mask=@n3, Emissive Color=@n4.RGB) {attributeSetTypes [#1, #2, #3]}) {materialFunction=&MF_TechArtOverrides}.Result) {materialFunction=/Game/Global/Materials/Character_Parent_2023/Functions/BaseFN_CharacterOpacity_2023}
 ]
}
