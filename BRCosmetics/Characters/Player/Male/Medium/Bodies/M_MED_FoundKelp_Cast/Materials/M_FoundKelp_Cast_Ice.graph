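# Translucent character material for the Found Kelp Cast male body, adding an ice look with refraction.
# The shared character PreFX function feeds MF_Ice_TechArt, whose attributes get opacity and refraction overrides before PostFX.
# Opacity blends from opaque to a Fresnel-plus-base value, then passes through MF_SortTranslucency.
# A custom HLSL spiral blur samples jittered scene color with quality-switched steps into BlurredColor, which has no visible link to the output.
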Material /BRCosmetics/Characters/Player/Male/Medium/Bodies/M_MED_FoundKelp_Cast/Materials/M_FoundKelp_Cast_Ice
graph {
 properties {
  allowTranslucentCustomDepthWrites=true
  bAutomaticallySetUsageInEditor=false
  bOutputTranslucentVelocity=true
  bUseMaterialAttributes=true
  bUsedWithMeshDeformer=true
  bUsedWithMorphTargets=true
  bUsedWithSkeletalMesh=true
  blendMode=BLEND_Translucent
  pixelDepthOffsetMode=PDOM_Legacy
  refractionMethod=RM_PixelNormalOffset
  shadingModel=MSM_FromMaterialExpression
  translucencyLightingMode=TLM_SurfacePerPixelLighting
  usedShadingModels=Default Lit | Subsurface | Clear Coat | Subsurface Profile | Cloth
 }
 nodes [
  MaterialExpressionViewSize_2 = ViewSize()
  n1 = Convert(None=@MaterialExpressionViewSize_2) {
   nodeName=Break Float2
   convertInputs [{
     type=Vector2
     expressionInput {
      expression=&self:MaterialExpressionViewSize_2
      inputName=None
      mask=0
      maskA=0
      maskB=0
      maskG=0
      maskR=0
      outputIndex=0
     }
    }]
   convertOutputs [{type=Scalar}, {type=Scalar}, {type=Vector2}]
  }
  ComponentMask(None=MaterialFunctionCall(Texture Dimensions=@n1) {materialFunction=&ScreenAlignedUVs}."X 100%, Y 100%") {g=true, r=true}
  n2 = MaterialFunctionCall() {materialFunction=&MF_CharacterParent_2023_PreFX}
  n3 = MaterialFunctionCall!(In=@n2.Result, InMask=@n2.Result, InFXMask=Constant() {r=1}) {materialFunction=/Game/Global/MaterialFunctions/TechArt/MF_Ice_TechArt}
  MaterialFunctionCall(
   Material=SetMaterialAttributes!(
    MaterialAttributes=GetMaterialAttributes(None=@n3.Result) {attributeGetTypes [#1, #2]}.MaterialAttributes
    Opacity=NamedRerouteUsage()
    Refraction=ScalarParameter() {defaultValue=1.05, parameterName=Refraction}
    Base Color
   ) {attributeSetTypes [#3, #4, #1]}
   M=@n2.M
   Pre Skinned Local Position=@n2.M
   Distance Blend=@n2.M
   Fresnel=@n2.M
  ) {materialFunction=&MF_CharacterParent_2023_PostFX}
  NamedRerouteDeclaration(None=ChannelMaskParameter(None=@n2.Tech Art Mask) {parameterName=TransparentIceMask}.R) {name=TransparentIceMask, nodeColor=<rgba 0.8984 0 1 1>}
  n4 = NamedRerouteUsage()
  n5 = ScalarParameter() {defaultValue=1.5, parameterName=ROUGHNESS POW}
  n6 = SmoothStep(Value=Power(Base=Fresnel() {baseReflectFraction=0, exponent=2}, Exp=ScalarParameter() {defaultValue=1, parameterName=FresnelPower})) {constMax=0.5}
  n7 = ScalarParameter() {defaultValue=0.5, parameterName=BaseOpacity}
  n8 = Saturate(None=Add(A=Power(Base=@n4, Exp=@n5) {constExponent=12}, B=Add(A=@n6, B=@n7)))
  n9 = NamedRerouteUsage()
  NamedRerouteDeclaration(None=MaterialFunctionCall!(Opacity=LinearInterpolate(B=@n8, Alpha=@n9) {constA=1}) {materialFunction=/Game/Global/Materials/MaterialGym/Materials/SortedOutline/MF_SortTranslucency}.Result) {name=Opacity_OUT, nodeColor=<rgba 0 1 0.5781 1>}
  NamedRerouteDeclaration(None=ChannelMaskParameter(None=@n2.M) {parameterName=AOChannel}.R) {name=AOMask, nodeColor=<rgba 1 0 0.8203 1>}
  NamedRerouteDeclaration(None=GetMaterialAttributes(None=@n2.Result) {attributeGetTypes [#2]}.Roughness) {name=Roughness_IN, nodeColor=<rgba 0 1 0.2969 1>}
  MaterialExpressionLinearInterpolate_2 = LinearInterpolate(
   A=ScalarParameter() {defaultValue=0.005, group=TransparentArmor, parameterName=SpiralBlur_Distance_Near}
   B=ScalarParameter() {defaultValue=0.001, group=TransparentArmor, parameterName=SpiralBlur_Distance_Far}
   Alpha=MaterialFunctionCall(Fade Length=ScalarParameter() {defaultValue=256, group=TransparentArmor, parameterName=SpiralBlur_FadingLength}) {materialFunction=&CameraDepthFade}.Result
  )
  n10 = Constant()
  MaterialExpressionMaterialFunctionCall_5 = MaterialFunctionCall!(Low=@n10, High=Constant() {r=4}, Epic=Constant() {r=8}, --------, Mobile=@n10, Medium=@n10) {materialFunction=/FortniteBaseMaterials/Functions/MF_QualitySwitch_Scalar}
  MaterialExpressionScalarParameter_21 = ScalarParameter() {defaultValue=1, parameterName=SpiralBlur_RadialSteps}
  MaterialExpressionConstant_4 = Constant() {r=1}
  MaterialExpressionTextureObject_5 = TextureObject() {samplerType=SAMPLERTYPE_LinearGrayscale, texture=/Engine/EngineMaterials/Good64x64TilingNoiseHighFreq}
  MaterialExpressionScalarParameter_22 = ScalarParameter() {defaultValue=0.618, parameterName=SpiralBlur_RadialOffset}
  n11 = MaterialFunctionCall() {materialFunction=&ObjectPivotPoint}
  n12 = CameraPositionWS()
  MaterialExpressionMultiply_23 = Multiply(
   A=ScalarParameter() {defaultValue=0.02, parameterName=SpiralBlur_RefractionAmount}
   B=Normalize(VectorInput=Subtract(A=@n11.Object Pivot Location, B=@n12))
  )
  n13 = Custom!(
   Distance=@MaterialExpressionLinearInterpolate_2
   DistanceSteps=@MaterialExpressionMaterialFunctionCall_5.Result
   RadialSteps=@MaterialExpressionScalarParameter_21
   TempAARotation=@MaterialExpressionConstant_4
   DistanceMask=@MaterialExpressionConstant_4
   Tex=@MaterialExpressionTextureObject_5
   TempAADistance=@MaterialExpressionConstant_4
   RadialOffset=@MaterialExpressionScalarParameter_22
   RefractionAmount=@MaterialExpressionMultiply_23
  ) {
   code="float3 CurColor = 0;\r\n\r\nfloat2 RawUV = MaterialFloat2(ScreenAlignedPosition(Parameters.ScreenPosition).xy);\r\n//float2 RefractedUV = BaseUV + (BaseUV.xy - 0.5) * RefractionAmount;\r\nfloat2 BaseUV = RawUV + (RawUV.xy - 0.5) * RefractionAmount;\r\nfloat2 NewUV = BaseUV;\r\nfloat StepSize = Distance / (int) DistanceSteps;\r\nfloat CurDistance = 0;\r\nfloat2 CurOffset = 0;\r\nfloat TwoPi = 6.283185;\r\nfloat Substep = 0;\r\nfloat2 ScenePixels=View.BufferSizeAndInvSize.xy*BaseUV;\r\nScenePixels+=View.TemporalAAParams.r;\r\nfloat2 RandomSamp = ((uint)(ScenePixels.x) + 2 * (uint)(ScenePixels.y)) % 5;\r\nRandomSamp+=Texture2DSample(Tex,TexSampler,ScenePixels);\r\nRandomSamp/=5;\r\nRandomSamp-=0.5;\r\nTempAARotation*=RandomSamp;\r\nTempAADistance*=StepSize*RandomSamp;\r\n\r\n\r\nint i=0;\r\nif (DistanceSteps<1)\r\n{\r\nreturn DecodeSceneColorForMaterialNode(NewUV);\r\n}\r\nelse\r\n{\r\nfloat2 BilinearOffset = View.BufferSizeAndInvSize.zw * 0.5f;\r\nfloat2 BilinearMinBufferUV = ViewportUVToBufferUV(float2(0.0f, 0.0f)) + BilinearOffset;\r\nfloat2 BilinearMaxBufferUV = ViewportUVToBufferUV(float2(1.0f, 1.0f)) - BilinearOffset;\r\n//CurDistance += 0.5*StepSize;\r\nwhile ( i < (int) DistanceSteps)\r\n{\r\n\r\n//CurDistance+=StepSize;\r\nfor (int j = 0; j < (int) RadialSteps; j++)\r\n{\r\nCurOffset.x = cos(TwoPi*((TempAARotation+Substep) / RadialSteps));\r\nCurOffset.y = sin(TwoPi*((TempAARotation+Substep) / RadialSteps));\r\nCurOffset *=DistanceMask;\r\nNewUV.x = BaseUV.x + (CurOffset.x * (CurDistance+(RandomSamp*TempAADistance)));\r\nNewUV.y = BaseUV.y + (CurOffset.y * (CurDistance+(RandomSamp*TempAADistance)));\r\nNewUV = clamp(NewUV, BilinearMinBufferUV, BilinearMaxBufferUV);\r\nCurColor += DecodeSceneColorForMaterialNode(NewUV);\r\n//CurDistance+=(StepSize+(TempAADistance))/RadialSteps;\r\nSubstep++;\r\n}\r\nCurDistance+=StepSize;\r\nSubstep+=RadialOffset;\r\ni++;\r\n}\r\nCurColor = CurColor / ((int)DistanceSteps*(int)RadialSteps);\r\nreturn CurColor;\r\n}"
   desc="float3 CurColor = 0;\r\nfloat2 BaseUV = MaterialFloat2(ScreenAlignedPosition(Parameters.ScreenPosition).xy);\r\nfloat2 NewUV = BaseUV;\r\nfloat StepSize = Distance / (int) DistanceSteps;\r\nfloat CurDistance = 0;\r\nfloat2 CurOffset = 0;\r\nfloat TwoPi = 6.283185;\r\nfloat Substep = 0;\r\nfloat2 ScenePixels=View.BufferSizeAndInvSize.xy*BaseUV;\r\nScenePixels+=View.TemporalAAParams.r;\r\nfloat2 RandomSamp = ((uint)(ScenePixels.x) + 2 * (uint)(ScenePixels.y)) % 5;\r\nRandomSamp+=Texture2DSample(Tex,TexSampler,ScenePixels);\r\nRandomSamp/=5;\r\nRandomSamp-=0.5;\r\nTempAARotation*=RandomSamp;\r\nTempAADistance*=StepSize*RandomSamp;\r\n\r\n\r\nint i=0;\r\nif (DistanceSteps<1)\r\n{\r\nreturn DecodeSceneColorForMaterialNode(NewUV);\r\n}\r\nelse\r\n{\r\nfloat2 BilinearOffset = View.BufferSizeAndInvSize.zw * 0.5f;\r\nfloat2 BilinearMinBufferUV = ViewportUVToBufferUV(float2(0.0f, 0.0f)) + BilinearOffset;\r\nfloat2 BilinearMaxBufferUV = ViewportUVToBufferUV(float2(1.0f, 1.0f)) - BilinearOffset;\r\n//CurDistance += 0.5*StepSize;\r\nwhile ( i < (int) DistanceSteps)\r\n{\r\n\r\n//CurDistance+=StepSize;\r\nfor (int j = 0; j < (int) RadialSteps; j++)\r\n{\r\nCurOffset.x = cos(TwoPi*((TempAARotation+Substep) / RadialSteps));\r\nCurOffset.y = sin(TwoPi*((TempAARotation+Substep) / RadialSteps));\r\nCurOffset *=DistanceMask;\r\nNewUV.x = BaseUV.x + (CurOffset.x * (CurDistance+(RandomSamp*TempAADistance)));\r\nNewUV.y = BaseUV.y + (CurOffset.y * (CurDistance+(RandomSamp*TempAADistance)));\r\nNewUV = clamp(NewUV, BilinearMinBufferUV, BilinearMaxBufferUV);\r\nCurColor += DecodeSceneColorForMaterialNode(NewUV);\r\n//CurDistance+=(StepSize+(TempAADistance))/RadialSteps;\r\nSubstep++;\r\n}\r\nCurDistance+=StepSize;\r\nSubstep+=RadialOffset;\r\ni++;\r\n}\r\nCurColor = CurColor / ((int)DistanceSteps*(int)RadialSteps);\r\nreturn CurColor;\r\n}"
   showCode=true
   inputs [
    {
     inputName=Distance
     input {expression=&self:MaterialExpressionLinearInterpolate_2}
    }
    {
     inputName=DistanceSteps
     input {expression=&self:MaterialExpressionMaterialFunctionCall_5}
    }
    {
     inputName=RadialSteps
     input {expression=&self:MaterialExpressionScalarParameter_21}
    }
    {
     inputName=TempAARotation
     input {expression=&self:MaterialExpressionConstant_4}
    }
    {
     inputName=DistanceMask
     input {expression=&self:MaterialExpressionConstant_4}
    }
    {
     inputName=Tex
     input {expression=&self:MaterialExpressionTextureObject_5}
    }
    {
     inputName=TempAADistance
     input {expression=&self:MaterialExpressionConstant_4}
    }
    {
     inputName=RadialOffset
     input {expression=&self:MaterialExpressionScalarParameter_22}
    }
    {
     inputName=RefractionAmount
     input {expression=&self:MaterialExpressionMultiply_23}
    }
   ]
  }
  NamedRerouteDeclaration(None=@n13) {name=BlurredColor, nodeColor=<rgba 0.0547 1 0 1>}
  ScalarParameter() {defaultValue=4, group=TransparentArmor, parameterName=SpiralBlur_Steps}
 ]
}
