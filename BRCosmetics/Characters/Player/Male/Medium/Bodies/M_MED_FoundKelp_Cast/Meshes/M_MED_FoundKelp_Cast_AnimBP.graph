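# Animation Blueprint for the FoundKelp_Cast male medium player body cosmetic, posing it from the attached parent mesh and adding secondary motion to its gear.
# Copy Pose From Mesh feeds the body's Control Rig, followed by three chained Anim Dynamics nodes for the right-side grenade and two pouch bones.
# Each simulates root-relative with angular springs, tight limits, reduced gravity and partial alpha blending.
# The event graph has no logic.
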AnimBlueprint /BRCosmetics/Characters/Player/Male/Medium/Bodies/M_MED_FoundKelp_Cast/Meshes/M_MED_FoundKelp_Cast_AnimBP
target_skeleton=&/M_MED_FoundKelp_Cast_Skeleton
graphs [
 AnimGraph {nodes [
   n1 = Control Rig(Source=Copy Pose From Mesh() {bUseAttachedParent=true}) {controlRigAssetReference {blueprintRigClass=&/M_MED_FoundKelp_Cast_CtrlRig.M_MED_FoundKelp_Cast_CtrlRig_C}}
   n2 = Anim Dynamics - Bone: dyn_grenade_01_r(ComponentPose=Local To Component(LocalPose=@n1), Alpha=0.86, ExternalForce="0,0,0") {
    angularDampingOverride=1
    angularSpringConstant=12000
    bAngularSpring=true
    bOverrideAngularDamping=true
    bUseGravityOverride=true
    gravityScale=0.35
    simulationSpace=RootRelative
    boundBone {boneName=dyn_grenade_01_r}
    componentAppliedLinearAccClamp=<7000 7000 7000>
    componentLinearAccScale=<0.5 0.5 0.5>
    componentLinearVelScale=<0.5 0.5 0.5>
    gravityOverride {z=-980}
    physicsBodyDefinitions [{
      boundBone {boneName=dyn_grenade_01_r}
      boxExtents=<20 30 30>
      constraintSetup {angularLimitsMax=<0 10 10>, angularLimitsMin=<0 -8 -15>, angularTarget=<1 0 0>}
      localJointOffset=<2.75 0 0>
     }]
   }
   n3 = Anim Dynamics - Bone: dyn_pouch_01_r(ComponentPose=@n2, Alpha=0.75, ExternalForce="0,0,0") {
    angularDampingOverride=1
    angularSpringConstant=8000
    bAngularSpring=true
    bOverrideAngularDamping=true
    bUseGravityOverride=true
    gravityScale=0.35
    simulationSpace=RootRelative
    boundBone {boneName=dyn_pouch_01_r}
    componentAppliedLinearAccClamp=<5000 5000 6000>
    componentLinearAccScale=<0.5 0.5 0.5>
    componentLinearVelScale=<0.5 0.5 0.5>
    gravityOverride {z=-980}
    physicsBodyDefinitions [{
      boundBone {boneName=dyn_pouch_01_r}
      boxExtents=<36 36 36>
      constraintSetup {angularLimitsMax=<0 7 7>, angularLimitsMin=<0 -7 -10>, angularTarget=<1 0 0>}
      localJointOffset=<2 0 0>
     }]
   }
   n4 = Anim Dynamics - Bone: dyn_pouch_02_r(ComponentPose=@n3, Alpha=0.8, ExternalForce="0,0,0") {
    angularDampingOverride=1
    angularSpringConstant=10000
    bAngularSpring=true
    bOverrideAngularDamping=true
    bUseGravityOverride=true
    gravityScale=0.35
    simulationSpace=RootRelative
    boundBone {boneName=dyn_pouch_02_r}
    componentAppliedLinearAccClamp=<6500 6500 6500>
    componentLinearAccScale=<0.65 0.65 0.65>
    componentLinearVelScale=<0.65 0.65 0.65>
    gravityOverride {z=-980}
    physicsBodyDefinitions [{
      boundBone {boneName=dyn_pouch_02_r}
      boxExtents=<29 29 29>
      constraintSetup {angularLimitsMax=<0 12 10>, angularLimitsMin=<0 -12 -15>, angularTarget=<1 0 0>}
      localJointOffset=<2.25 0 0>
     }]
   }
   Output Pose(Result=Component To Local(ComponentPose=@n4)) {layerGroup=None}
  ]}
 EventGraph {nodes [Event BlueprintUpdateAnimation()]}
]
variables []
